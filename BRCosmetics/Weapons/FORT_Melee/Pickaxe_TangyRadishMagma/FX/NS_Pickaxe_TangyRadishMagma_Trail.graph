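# Niagara ribbon trail for the TangyRadishMagma pickaxe, with two color variants selected by a Variant parameter.
# MainTrail spawns two socket-driven anim trails with their own color and fade by spawn group, while TrailTip adds a brighter accent ribbon with width tapering over a curve.
# Both emitters use short-lived particles, tint by a time-of-day color, and multiply alpha by an emitter-age curve that cuts the trail off shortly after it starts.
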NiagaraSystem /BRCosmetics/Weapons/FORT_Melee/Pickaxe_TangyRadishMagma/FX/NS_Pickaxe_TangyRadishMagma_Trail
dependencies {
 usedDataInterfaces [&NiagaraDataInterfaceColorCurve, &NiagaraDataInterfaceCurve, &NiagaraDataInterfaceSocketReader]
 usedDynamicInputs [&MakeIntFromFloat, &FNTimeOfDayColor]
 usedModules [
  &self:SystemSpawnScript.SystemScriptSource.SystemScriptGraph.NiagaraNodeAssignment_1.SetVariables_#1
  &self:SystemSpawnScript.SystemScriptSource.SystemScriptGraph.NiagaraNodeAssignment_2.SetVariables_#2
  &self:SystemSpawnScript.SystemScriptSource.SystemScriptGraph.NiagaraNodeAssignment_0.SetVariables_#3
  &SystemState
 ]
}
emitters {
 MainTrail {
  emitterSpawnScript [&FN_RibbonIDOffsetUtility()]
  emitterUpdateScript [
   &EmitterState()
   &FN_SpawnAnimTrail(Number of Anim Trails=2)
  ]
  particleSpawnScript [
   &FN_InitializeAnimTrailBoneIndecies()
   &FN_InitializeAnimTrailBoneIndecies(AnimTrail Selection=1) {module=FN_InitializeAnimTrailBoneIndecies001}
   &FN_InitializeAnimTrail(
    AnimTrailWidth=&MergeHelper_FloatSwitch(0=1, 1=1, 2=?, 3=?, Index=&SpawnGroup(), Num=::NewEnumerator0)
    Color=&MergeHelper_ColorSwitch(
     0=&MergeHelper_ColorSwitch(0=<rgba 0.13 0.1475 0.2 1>, 1=<rgba 1 0 0.1 1>, 2=?, 3=?, Index=&SpawnGroup(), Num=::NewEnumerator0)
     1=&MergeHelper_ColorSwitch(0=<rgba 1 0.4 0.8 0.3>, 1=<rgba 0.8125 0.9219 1.25 1>, 2=?, 3=?, Index=&SpawnGroup(), Num=::NewEnumerator0)
     2=?
     3=?
     Index=$System.Variant
     Num=::NewEnumerator0
    )
    Lifetime=0.16
   )
  ]
  particleUpdateScript [
   &SolveForcesAndVelocity(Clamp Velocity=false, Limit Acceleration=false)
   &FN_UpdateAnimTrail() {module=FN_UpdateAnimTrail001}
   &ScaleColor(
    Scale Alpha=&FloatFromCurve(
     FloatCurve=DataInterface() {propertyValues {Curve {keys [
         {arriveTangent=0.1441, interpMode=RCIM_Cubic, leaveTangent=-2.4459, tangentMode=RCTM_Break, time=0, value=1}
         {arriveTangent=-0.0531, interpMode=RCIM_Cubic, tangentMode=RCTM_Break, time=1, value=0}
        ]}}}
     CurveIndex=&Multiply_Float(
      A=$Particles.NormalizedAge
      B=&MergeHelper_FloatSwitch(0=1, 1=3.5, 2=?, 3=?, Index=&SpawnGroup(), Num=::NewEnumerator0)
     )
     Scale Curve=&FloatFromCurve(
      FloatCurve=DataInterface() {propertyValues {Curve {keys [<at 0.25 1>,<at 0.32 0>]}}}
      CurveIndex=$Emitter.Age
      Scale Curve=&MergeHelper_FloatSwitch(0=1.2, 1=1.4, 2=?, 3=?, Index=&SpawnGroup(), Num=::NewEnumerator0)
     )
    )
    Scale RGB=&Multiply_Vector(
     A=&MakeVectorFromLinearColorRGB(Color=$System.ToDColor)
     B=&MergeHelper_VectorSwitch(
      0=<1 1 1>
      1=&MakeVectorFromLinearColorRGB(Color=&ColorFromCurve(ColorCurve=DataInterface() {propertyValues {
          AlphaCurve {keys [<at 0 1 RCIM_Cubic>,<at 1 0 RCIM_Cubic>]}
          BlueCurve {keys [<at 0 1 RCIM_Cubic>,<at 0.1882 0.5>,<at 0.789 0.05 RCIM_Cubic>]}
          GreenCurve {keys [<at 0 1 RCIM_Cubic>,<at 0.1882 0.5>,<at 0.789 0.05 RCIM_Cubic>]}
          RedCurve {keys [<at 0 1 RCIM_Cubic>,<at 0.1882 0.5>,<at 0.789 0.05 RCIM_Cubic>]}
         }}))
      2=?
      3=?
      Index=&SpawnGroup()
      Num=::NewEnumerator0
     )
    )
    ScaleA=true
    ScaleRGB=true
   )
  ]
  rendererClasses [&NiagaraRibbonRendererProperties]
  renderers [{rendererClass=&NiagaraRibbonRendererProperties}]
 }
 TrailTip {
  emitterSpawnScript [&FN_RibbonIDOffsetUtility()]
  emitterUpdateScript [
   &EmitterState()
   &FN_SpawnAnimTrail(AnimTrailWidthMode=::FromSecond)
  ]
  particleSpawnScript [
   &FN_InitializeAnimTrailBoneIndecies(First Bone or Socket Index=2, Second Bone or Socket Index=2)
   &FN_InitializeAnimTrail(
    AnimTrailWidth=2
    Color=&MergeHelper_ColorSwitch(0=<rgba 2.5 0 0.1667 1>, 1=<rgba 2.5 2.25 2.2917 1>, 2=?, 3=?, Index=$System.Variant, Num=::NewEnumerator0)
    Lifetime=0.16
   )
  ]
  particleUpdateScript [
   &SolveForcesAndVelocity(Clamp Velocity=false, Limit Acceleration=false)
   &FN_UpdateAnimTrail() {module=FN_UpdateAnimTrail001}
   &ScaleColor(
    Scale Alpha=&FloatFromCurve(
     FloatCurve=DataInterface() {propertyValues {Curve {keys [
         {arriveTangent=0.1441, interpMode=RCIM_Cubic, leaveTangent=-2.4459, tangentMode=RCTM_Break, time=0, value=1}
         {arriveTangent=-0.0531, interpMode=RCIM_Cubic, tangentMode=RCTM_Break, time=1, value=0}
        ]}}}
     Scale Curve=&FloatFromCurve(FloatCurve=DataInterface() {propertyValues {Curve {keys [<at 0.25 1>,<at 0.32 0>]}}}, CurveIndex=$Emitter.Age, Scale Curve=0.7)
    )
    Scale RGB=&Multiply_Vector(
     A=&MakeVectorFromLinearColorRGB(Color=$System.ToDColor)
     B=&MakeVectorFromLinearColorRGB(Color=&ColorFromCurve(ColorCurve=DataInterface() {propertyValues {
         AlphaCurve {keys [<at 0 1 RCIM_Cubic>,<at 1 0 RCIM_Cubic>]}
         BlueCurve {keys [<at 0 1 RCIM_Cubic>,<at 0.1251 0.5>,<at 0.4991 0.05 RCIM_Cubic>]}
         GreenCurve {keys [<at 0 1 RCIM_Cubic>,<at 0.1251 0.5>,<at 0.4991 0.05 RCIM_Cubic>]}
         RedCurve {keys [<at 0 1 RCIM_Cubic>,<at 0.1251 0.5>,<at 0.4991 0.05 RCIM_Cubic>]}
        }}))
    )
    ScaleA=true
    ScaleRGB=true
   )
   &ScaleRibbonWidth(Ribbon Width Scale=&FloatFromCurve(FloatCurve=DataInterface() {propertyValues {Curve {keys [<at 0 0.5>,<at 0.1 1>,<at 1 0.3>]}}}))
  ]
  rendererClasses [&NiagaraRibbonRendererProperties]
  renderers [{rendererClass=&NiagaraRibbonRendererProperties}]
 }
}
summary {
 systemName=NS_Pickaxe_TangyRadishMagma_Trail
 emitters [
  {emitterName=TrailTip, rendererClasses [&NiagaraRibbonRendererProperties]}
  {emitterName=MainTrail, rendererClasses [&NiagaraRibbonRendererProperties]}
 ]
 userVariables [
  User.RibbonIDOffset {type=NiagaraInt32}
  User.Variant {type=NiagaraFloat, defaultValue=0, description=""}
 ]
}
